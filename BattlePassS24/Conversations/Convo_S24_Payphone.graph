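# Conversation database for a Season 24 payphone interaction, entered through the Conversation.Entry.Season24.Payphone tag.
# The entry leads to a speech node saying the line suddenly goes silent, which continues to a final bold "*CLICK*" speech.
# That last node carries a "Hang up" choice, so the conversation is a short linear exchange with no branching.
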ConversationDatabase /BattlePassS24/Conversations/Convo_S24_Payphone
entry_points [{tag=Conversation.Entry.Season24.Payphone, destinations [#1]}]
nodes [
 ConversationEntryPointNode_4 {
  id=#1
  class=ConversationEntryPointNode
  outputs [#2]
  properties {entryTag {tagName=Conversation.Entry.Season24.Payphone}}
 }
 Speech_C_11 {
  id=#2
  class=Speech_C
  outputs [#3]
  properties {
   softTaskIcon=/Game/UI/Foundation/Textures/Icons/Replay/T-Icon-Replay-Plus-L
   generalConfig {defaultMessage=(The line suddenly goes silent.), defaultMetadataParameters [1 {value=Default}]}
  }
 }
 Speech_C_12 {
  id=#3
  class=Speech_C
  outputs []
  properties {generalConfig {defaultMessage="<b> *CLICK* </>"}}
  sub_nodes [ConversationChoiceNode_11 {class=ConversationChoiceNode, properties {defaultChoiceDisplayText=Hang up}}]
 }
]
